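# Translucent unlit neon material for the SliceVine skydiving contrail Niagara effect, drawing a particle-colored glowing line.
# UV smoothsteps form a core and soft glow, faded by particle alpha, camera distance and depth (non-mobile).
# World position offset combines SplineThicken with a vertex-red-masked unroll, driven by dynamic parameters, that shifts and radially widens the mesh.
# An optional time sine, phased by vertex alpha, modulates both opacity and thickening.
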Material /BRCosmetics/Athena/Effects/Particles/Skydiving/NS_Contrail_SliceVine/M_SliceVine_NeonMaterial
graph {
 properties {
  bAutomaticallySetUsageInEditor=false
  bEnableResponsiveAA=true
  bUsedWithNiagaraMeshParticles=true
  bUsedWithNiagaraRibbons=true
  bUsedWithNiagaraSprites=true
  blendMode=BLEND_Translucent
  shadingModel=MSM_Unlit
  twoSided=true
  usedShadingModels=Unlit
 }
 nodes [
  n1 = Time()
  n2 = ScalarParameter() {defaultValue=0.7, parameterName=RandomOpacitySpeed}
  n3 = Sine(None=Add(A=VertexColor().A, B=Multiply(A=@n1, B=@n2)))
  n4 = LinearInterpolate(
   B=MaterialFunctionCall(
    Input=ConstantBiasScale(None=@n3)
    Input Low=Constant()
    Input High=Constant() {r=1}
    Target Low=ScalarParameter() {defaultValue=0.3, parameterName=RandomOpacityMin}
   ) {materialFunction=&RemapValueRange}.Result
   Alpha=ScalarParameter() {parameterName=UseRandomOpacity}
  ) {constA=1}
  n5 = Multiply(A=ScalarParameter() {parameterName=SplineWidth}, B=DynamicParameter() {paramNames [Unroll, UnrollMultiplier, WidthMult, Param4]}.WidthMult)
  n6 = TextureCoordinate()
  n7 = StaticBool()
  n8 = VertexColor()
  n9 = DynamicParameter() {paramNames [Unroll, UnrollMultiplier, WidthMult, Param4]}
  n10 = Add(A=Multiply(A=OneMinus(None=@n8.R)) {constB=-0.25}, B=@n9.Unroll)
  n11 = Saturate(None=Multiply(A=@n8.R, B=Saturate(None=@n10)))
  n12 = DynamicParameter() {parameterIndex=1, paramNames [UnrollX,UnrollY,UnrollZ,Param4]}
  n13 = MaterialFunctionCall(X=@n12.UnrollX, Y=@n12.UnrollX, Z=@n12.UnrollX) {materialFunction=&MakeFloat3}
  n14 = Multiply(A=@n11, B=Multiply(A=@n9.UnrollMultiplier, B=@n13.Result))
  n15 = Subtract(
   A=WorldPosition() {worldPositionShaderOffset=WPT_ExcludeAllShaderOffsets}.XYZ
   B=TransformPosition(None=Constant3Vector()) {transformSourceType=TRANSFORMPOSSOURCE_Instance, transformType=TRANSFORMPOSSOURCE_World}
  )
  n16 = Constant3Vector() {constant=<rgba 1 1 0 0>}
  n17 = VertexColor()
  n18 = Multiply(
   A=Normalize(VectorInput=Multiply(A=@n15, B=@n16))
   B=Multiply(A=Add(B=@n17.R) {constA=1}, B=ScalarParameter() {defaultValue=7, parameterName=UnrollWiden})
  )
  Add(
   A=Multiply(
    A=@n4
    B=MaterialFunctionCall(WidthBase=@n5, WidthTip=@n5, UVs for Projection=@n6, UVs for Thickness=@n6, UVs for texturing=@n6, Expand U or V UV Channel=@n7) {materialFunction=&SplineThicken}.WorldPositionOffset
   ) {constA=1}
   B=Add(A=Transform(None=@n14) {transformSourceType=TRANSFORMSOURCE_World, transformType=TRANSFORM_Local}, B=Multiply(A=@n11, B=@n18))
  )
  Constant3Vector() {constant=<rgba 0 0 1 0>}
  n19 = ParticleColor()
  MaterialFunctionCall(Emissive=@n19.RGB) {materialFunction=&MF_FX_EmissiveAdjustments}
  n20 = ScalarParameter() {defaultValue=0.07, parameterName=CoreWidth}
  n21 = OneMinus(None=Add(A=Divide(A=@n20)) {constB=0.5})
  n22 = ComponentMask(None=TextureCoordinate()) {g=true}
  n23 = Min(A=@n22, B=OneMinus(None=@n22))
  n24 = SmoothStep(Min=@n21, Max=Add(A=@n21, B=ScalarParameter() {defaultValue=0.05, parameterName=CoreSoftness}), Value=@n23) {constMax=0.5, constMin=0.25}
  n25 = ScalarParameter() {defaultValue=0.7, parameterName=GlowWidth}
  n26 = OneMinus(None=Add(A=Divide(A=@n25)) {constB=0.5})
  n27 = Add(A=@n26, B=ScalarParameter() {defaultValue=1, parameterName=GlowSoftness})
  n28 = ComponentMask(None=TextureCoordinate()) {r=true}
  n29 = SmoothStep(Value=Min(A=@n28, B=OneMinus(None=@n28))) {constMax=0.3}
  n30 = Multiply(
   A=Power(
    Base=SmoothStep(Min=@n26, Max=@n27, Value=@n23) {constMax=0.5, constMin=0.25}
    Exp=ScalarParameter() {defaultValue=1.5, parameterName=GlowPower}
   )
   B=Multiply(A=Saturate(None=@n29), B=ScalarParameter() {defaultValue=0.5, parameterName=GlowStrength})
  )
  n31 = MaterialFunctionCall(
   Fade Length=ScalarParameter() {defaultValue=200, parameterName=CameraFadeLength}
   Fade Offset=ScalarParameter() {defaultValue=100, parameterName=CameraFadeOffset}
  ) {materialFunction=&CameraDepthFade}
  n32 = DepthFade(FadeDistance=ScalarParameter() {defaultValue=50, parameterName=DepthFade})
  n33 = ShadingPathSwitch(Default=@n32, Deferred=@n32, Mobile=Constant() {r=1}) {desc=Migrated from FeatureLevelSwitch}
  Multiply(A=Multiply(A=@n19.A, B=Max(A=@n24, B=@n30)), B=Multiply(A=@n4, B=Multiply(A=@n31.Result, B=@n33)) {constA=1})
  Normalize(VectorInput=@n13.Result)
  Transform(None=@n13.Result) {transformSourceType=TRANSFORMSOURCE_Instance}
 ]
}
