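# Green vine-like skydiving contrail cosmetic attached to two skeletal mesh bones or sockets.
# NE_Balls draws a glowing orb with light at each point, while Flares adds sprites scaled by the smoothed dive alpha.
# NE_PlexusTrail reads the balls and recycles persistent particles by looping progress, pulling them toward noisy, velocity-influenced goals to form layered ribbons and meshes.
# Dive alpha also drives trail length, width and brightness, and front-end flags swap colors and directions.
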NiagaraSystem /BRCosmetics/Athena/Effects/Particles/Skydiving/NS_Contrail_SliceVine/NS_Contrail_SliceVine
dependencies {
 usedDataInterfaces [
  &NiagaraDataInterfaceColorCurve
  &NiagaraDataInterfaceParticleRead
  &NiagaraDataInterfaceCurve
  &NiagaraDataInterfaceCamera
  &NiagaraDataInterfaceSkeletalMesh
 ]
 usedDynamicInputs [/Niagara/DynamicInputs/Lerp/SmoothLerpOverTime_Float]
 usedModules [
  &self:SystemSpawnScript.SystemScriptSource.SystemScriptGraph.NiagaraNodeAssignment_1.SetVariables_#1
  &self:SystemSpawnScript.SystemScriptSource.SystemScriptGraph.NiagaraNodeAssignment_2.SetVariables_#2
  &SystemState
  &self:SystemSpawnScript.SystemScriptSource.SystemScriptGraph.NiagaraNodeAssignment_0.SetVariables_#3
 ]
}
emitters {
 Flares {
  emitterUpdateScript [
   &EmitterState()
   &SpawnRate(SpawnRate=5, Use Spawn Probability=false)
  ]
  particleSpawnScript [
   &PartitionParticles(Write Partition to Ribbon ID=::NewEnumerator0)
   &InitializeParticle(Color=<rgba 2 2 2 1>, Sprite Size Mode=::NewEnumerator3, Uniform Sprite Size=100, UsePositionOffset=true, Write Material Random=true)
   &CameraOffset(Camera Offset Amount=4)
  ]
  particleUpdateScript [
   &ParticleState()
   &SkeletalMeshLocation(
    Bone / Socket Index=$Particles.PartitionParticles.Partition
    Bone Sampling Mode=::NewEnumerator4
    MaskSpawnGroup=false
    Orientation Sampling=::NewEnumerator1
    Sampled Position Offset=&Lerp_Vector(A=<7 0 0>, B=<-7 0 1>, Alpha=&MakeFloatFromInt(Int=$Particles.PartitionParticles.Partition))
    Skeletal Mesh=$System.Skeletal Mesh
    Use Sampled Mesh Attribute as Ribbon ID=false
   )
   &ScaleColor(
    Scale Alpha=&FloatFromCurve(
     FloatCurve=DataInterface() {propertyValues {Curve {keys [
         <at 0 0 RCIM_Cubic>
         {arriveTangent=-0.4225, interpMode=RCIM_Cubic, leaveTangent=-0.4225, tangentMode=RCTM_User, time=0.5, value=0.5}
         {arriveTangent=-0.0075, interpMode=RCIM_Cubic, leaveTangent=-0.0075, tangentMode=RCTM_User, time=1, value=0}
        ]}}}
     Scale Curve=$System.SmoothDiveAlpha
    )
    Scale RGB=&MakeVectorFromLinearColorRGB(Color=&Multiply_LinearColorByFloat(
      Linear Color=&ColorFromCurve(ColorCurve=DataInterface() {propertyValues {
         AlphaCurve {keys [<at 0 1 RCIM_Cubic>]}
         BlueCurve {keys [<at 0.0426 1.2111 RCIM_Cubic>, {arriveTangent=-1.5584, interpMode=RCIM_Cubic, leaveTangent=-1.5584, time=0.346, value=0.0933}, <at 0.8197 0>]}
         GreenCurve {keys [<at 0.0426 3 RCIM_Cubic>, {arriveTangent=-3.0882, interpMode=RCIM_Cubic, leaveTangent=-3.0882, time=0.346, value=1}, <at 0.8197 0.6>]}
         RedCurve {keys [<at 0.0426 1.3002 RCIM_Cubic>, {arriveTangent=-1.5918, interpMode=RCIM_Cubic, leaveTangent=-1.5918, time=0.346, value=0}, <at 0.8197 0.0631>]}
        }})
      Float=1
     ))
    ScaleA=true
    ScaleRGB=true
   )
   &ScaleSpriteSize(
    Scale Sprite Size Mode=::NewEnumerator0
    Uniform Scale Factor=&FloatFromCurve(
     FloatCurve=DataInterface() {propertyValues {Curve {keys [
         {arriveTangent=2.0201, interpMode=RCIM_Cubic, leaveTangent=2.0201, tangentMode=RCTM_User, time=0, value=0}
         {arriveTangent=-0.026, interpMode=RCIM_Cubic, leaveTangent=-0.026, tangentMode=RCTM_User, time=0.6, value=1}
        ]}}}
     Scale Curve=$System.SmoothDiveAlpha
    )
   )
   &SolveForcesAndVelocity(Clamp Velocity=false, Limit Acceleration=false)
   &DynamicMaterialParameters(
    Index 0 Param 1=1
    Index 0 Param 2=1
    Index 0 Param 3=1
    Index 0 Param 4=1
    Module.Index0Param0Write=true
    Module.Index0Param1Write=true
    Module.Index0Param2Write=true
    Module.Index0Param3Write=true
   )
   &ScaleColor(
    Scale RGB=&MakeCustomVectorFromBool(
     Bool=&Boolean_OrOperation(Bool1=$User.bIsFrontEnd, Bool2=$User.bIsFrontEndPreview)
     True Vector=<0.5 0.5 0.5>
     False Vector=<1 0 0>
    )
    ScaleA=false
    ScaleRGB=true
   ) {module=ScaleColor001}
  ]
  rendererClasses [&NiagaraSpriteRendererProperties]
  renderers [{rendererClass=&NiagaraSpriteRendererProperties}]
 }
 NE_Balls {
  emitterUpdateScript [
   &EmitterState(
    Inactive Response=::NewEnumerator1
    Life Cycle Mode=::NewEnumerator1
    Loop Behavior=::NewEnumerator1
    Loop Duration Mode=::NewEnumerator1
    UseLoopDelay=false
   )
   &SpawnBurst_Instantaneous(Spawn Count=2, Use Spawn Probability=false, UseLoopCountLimit=false)
  ]
  particleSpawnScript [
   &PartitionParticles(
    ID to Partition:NiagaraInt32=&ReturnExecIndex()
    Number Of Partitions=6
    Write Partition to Ribbon ID=::NewEnumerator0
   )
   &InitializeParticle(
    Color Mode=::NewEnumerator0
    Lifetime=5
    Mesh Scale Mode=::NewEnumerator3
    Mesh Uniform Scale=0.6
    Sprite Size Mode=::NewEnumerator3
    Uniform Sprite Size=70
    UsePositionOffset=false
    Write Material Random=true
   )
   /SetVariables() {isSetParameters=true}
  ]
  particleUpdateScript [
   &ParticleState(Kill Particles When Lifetime Has Elapsed=false)
   /SetVariables(Particles.RandomOffset:Vector3f=&MakeVector(
     X=&Sine(
      Normalized Angle=&Add_Float(A=$Emitter.Age, B=&ReturnNormalizedExecIndex(Normalized Index Scale=2))
      Period=3
      Scale=1
      Bias=0
     )
     Y=&Cosine(
      Normalized Angle=&Add_Float(A=$Emitter.Age, B=&ReturnNormalizedExecIndex())
      Period=3
      Scale=1
      Bias=0
     )
     Z=0
    )) {isSetParameters=true}
   &SkeletalMeshLocation(
    Bone / Socket Index=$Particles.PartitionParticles.Partition
    Bone Sampling Mode=::NewEnumerator4
    MaskSpawnGroup=false
    Orientation Sampling=::NewEnumerator1
    Sampled Position Offset=&Lerp_Vector(A=<7 0 0>, B=<-7 0 1>, Alpha=&MakeFloatFromInt(Int=$Particles.PartitionParticles.Partition))
    Skeletal Mesh=$System.Skeletal Mesh
    Use Sampled Mesh Attribute as Ribbon ID=false
   )
   &ScaleMeshSize(Scale Factor=&VectorFromFloat(Value=&Lerp_Float(A=1, B=1, Alpha=$System.SmoothDiveAlpha)))
   &UpdateMeshOrientation(
    Rotation Rate=&Lerp_Float(A=0.6, B=1.2, Alpha=$System.SmoothDiveAlpha)
    Rotation Vector=&RandomVector(Vector Scale=1)
   )
   &Color(Color=&Multiply_LinearColorByFloat(
     Linear Color=&ColorFromCurve(ColorCurve=DataInterface() {propertyValues {
        AlphaCurve {keys [<at 0 1 RCIM_Cubic>]}
        BlueCurve {keys [<at 0 0.1719 RCIM_Cubic>]}
        GreenCurve {keys [<at 0 1.5 RCIM_Cubic>]}
        RedCurve {keys [<at 0 0.1719 RCIM_Cubic>]}
       }})
     Float=&Lerp_Float(A=1, B=1.5, Alpha=$System.SmoothDiveAlpha)
    ))
   &DynamicMaterialParameters(
    Index 0 Param 1=2
    Module.Index0Param0Write=true
    Module.Index0Param1Write=true
    Module.Index0Param2Write=true
    Module.Index0Param3Write=true
   )
   /SetVariables(Particles.FlareColor=&Multiply_LinearColorByFloat(Linear Color=<rgba 0 1 0.1246 0.7>, Float=&Lerp_Float(A=1, B=1.7, Alpha=$System.SmoothDiveAlpha))) {isSetParameters=true}
   /SetVariables(Particles.FlareSize:Vector2f=&Multiply_Vector2D_ByFloat(Float=0.6, Vector2D=$Particles.SpriteSize)) {isSetParameters=true}
   /SetVariables(Particles.GlowColor=&Multiply_LinearColorByFloat(Linear Color=<rgba 0 1 0.1246 0.2>, Float=&Lerp_Float(A=1, B=1.3, Alpha=$System.SmoothDiveAlpha))) {isSetParameters=true}
   /SetVariables(Particles.GlowSize=&Multiply_Vector2D_ByFloat(Float=1.7, Vector2D=$Particles.SpriteSize)) {isSetParameters=true}
   /SetVariables(Particles.DecalDynMat:Vector4f=<3 0 0 0>) {isSetParameters=true}
   /SetVariables(Particles.DecalColor:LinearColor=&MultiplyLinearColor(Linear Color=<rgba 0.2042 2 0 1>, Scale Factor=&MakeVector4_FromLinearColor(LinearColor=$Particles.Color))) {isSetParameters=true}
   /SetVariables(Particles.DecalCameraOffset=20) {isSetParameters=true}
   /SetVariables(Particles.GlowDecalSize:Vector2f=&Multiply_Vector2D_ByFloat(Float=0.4, Vector2D=$Particles.SpriteSize)) {isSetParameters=true}
   &FN_Light_Attributes(
    Color=$Particles.Color
    Create Color Attribute=true
    Create Exponent Attribute=true
    Create Radius Attribute=true
    Exponent=1
    Radius=45
    Scale Color=&Lerp_Float(A=5, B=7, Alpha=$System.SmoothDiveAlpha)
   )
   /SetVariables(Particles.SphereDynMat:Vector4f=&MakeVector4(X=0, Y=0, Z=&Lerp_Float(A=0.75, B=1, Alpha=$System.SmoothDiveAlpha), W=0)) {isSetParameters=true}
   /SetVariables(Particles.Color=&MakeCustomLinearColorFromBool(
     Bool=&Boolean_OrOperation(Bool1=$User.bIsFrontEnd, Bool2=$User.bIsFrontEndPreview)
     True Color=&ColorFromCurve(ColorCurve=DataInterface() {propertyValues {
        AlphaCurve {keys [<at 0 1 RCIM_Cubic>]}
        BlueCurve {keys [<at 0 0.1146 RCIM_Cubic>]}
        GreenCurve {keys [<at 0 1 RCIM_Cubic>]}
        RedCurve {keys [<at 0 0.1146 RCIM_Cubic>]}
       }})
     False Color=$Particles.Color
    )) {isSetParameters=true}
  ]
  rendererClasses [
   &NiagaraSpriteRendererProperties
   &NiagaraMeshRendererProperties
   &NiagaraLightRendererProperties
  ]
  renderers [
   {rendererClass=&NiagaraSpriteRendererProperties}
   {rendererClass=&NiagaraMeshRendererProperties, rendererIndex=1}
   {rendererClass=&NiagaraLightRendererProperties, rendererIndex=2}
   {rendererClass=&NiagaraSpriteRendererProperties, rendererIndex=3}
   {rendererClass=&NiagaraSpriteRendererProperties, rendererIndex=4}
  ]
 }
 NE_PlexusTrail {
  emitterSpawnScript [/SetVariables(Emitter.ParticlePerRibbon=20, Emitter.RibbonNumber=2) {isSetParameters=true}]
  emitterUpdateScript [
   &EmitterState(
    Inactive Response=::NewEnumerator1
    Life Cycle Mode=::NewEnumerator1
    Loop Behavior=::NewEnumerator1
    Loop Duration Mode=::NewEnumerator1
    UseLoopDelay=false
   )
   &SpawnBurst_Instantaneous(
    Spawn Count=&Multiply_Int(A=$Emitter.RibbonNumber, B=$Emitter.ParticlePerRibbon)
    Use Spawn Probability=false
    UseLoopCountLimit=false
   )
  ]
  particleSpawnScript [
   &PartitionParticles(
    ID to Partition:NiagaraInt32=&ReturnExecIndex()
    Number Of Partitions=$Emitter.RibbonNumber
    Particles Per Partition:NiagaraInt32=$Emitter.ParticlePerRibbon
    Partition Mode:ENiagara_PartitionMode=::NewEnumerator1
    Write Partition to Ribbon ID=::NewEnumerator0
   )
   &InitializeParticle(
    Color=&RandomRangeLinearColor(Minimum=<rgba 1 1 1 1>, Maximum=<rgba 1 1 1 0.5>)
    Lifetime=1.7
    Mesh Scale Mode=::NewEnumerator3
    Mesh Uniform Scale=&RandomRangeFloat(Minimum=0.3, Maximum=0.4)
    Ribbon Width=&RandomRangeFloat(Minimum=8, Maximum=8)
    Ribbon Width Mode=::NewEnumerator1
    Sprite Size Mode=::NewEnumerator3
    Uniform Sprite Size=&RandomRangeFloat(Minimum=1, Maximum=2)
    UsePositionOffset=false
    Write Material Random=true
   )
   /SetVariables() {isSetParameters=true}
   /SetVariables(Particles.ProgressAlpha:NiagaraFloat=$Particles.PartitionParticles.Normalized Index In Partition) {isSetParameters=true}
   /SetVariables() {isSetParameters=true}
   /SetVariables(Particles.DebugOffset=10) {isSetParameters=true}
   &SkeletalMeshLocation(
    Bone / Socket Index=$Particles.PartitionParticles.Partition
    Bone Sampling Mode=::NewEnumerator4
    MaskSpawnGroup=false
    Skeletal Mesh=$System.Skeletal Mesh
    Use Sampled Mesh Attribute as Ribbon ID=false
   ) {module=SkeletalMeshLocation001}
   &InitialMeshOrientation(Use Rotation Vector=false)
   /SetVariables() {isSetParameters=true}
  ]
  particleUpdateScript [
   &ParticleState(Kill Particles When Lifetime Has Elapsed=false)
   &SkeletalMeshLocation(
    Bone / Socket Index=$Particles.PartitionParticles.Partition
    Bone Sampling Mode=::NewEnumerator4
    MaskSpawnGroup=false
    Orientation Sampling=::NewEnumerator1
    Position Sampling=::NewEnumerator2
    Sampled Position Offset=&Lerp_Vector(A=<7 0 0>, B=<-7 0 1>, Alpha=&MakeFloatFromInt(Int=$Particles.PartitionParticles.Partition))
    Skeletal Mesh=$System.Skeletal Mesh
    Use Sampled Mesh Attribute as Ribbon ID=false
   )
   &self:RandomFloatPerCycle(
    CycleAlpha:NiagaraFloat=$Particles.ProgressAlpha
    RandomFloat:NiagaraFloat=&RandomRangeFloat(Minimum=0, Maximum=1, Evaluation Type=::NewEnumerator1)
    ResetTimer=0.99
   ) {module=RandomFloatPerCycle}
   &self:GetBallPos(Particle Attribute Reader=DataInterface() {propertyValues {EmitterBinding {bindingMode=Other, emitterName=NE_Balls}}}) {module=GetBallPos}
   /SetVariables(Particles.ProgressAlpha:NiagaraFloat=&ModuloFloat(
     A=&Add_Float(
      A=&Add_Float(
       A=$Particles.ProgressAlpha
       B=&Multiply_Float(A=1, B=&Lerp_Float(A=0.015, B=0.02, Alpha=$System.SmoothDiveAlpha))
      )
      B=0
     )
     B=1
    )) {isSetParameters=true}
   /Niagara/Modules/Update/Utility/VectorNoiseValue(
    Noise Frequency=7
    Pan Noise=true
    Pan Noise Field:Vector3f=<0 0 1.5>
    Sample Position:NiagaraPosition=&TransformPosition(Position=$Particles.Position, Source Space=::World, Destination Space=::Local)
   )
   /SetVariables(Particles.RandomOffset:Vector3f=&Multiply_VectorByFloat(
     Vector=&AddVectorToPosition(
      Position=&Multiply_Vector(
       A=&RandomVector(Vector Scale=&Multiply_Float(
         A=&Lerp_Float(A=15, B=20, Alpha=$System.SmoothDiveAlpha)
         B=&FloatFromCurve(
          FloatCurve=DataInterface() {propertyValues {Curve {keys [
              {arriveTangent=-0.0811, interpMode=RCIM_Cubic, leaveTangent=-0.0811, tangentMode=RCTM_User, time=0, value=1}
              {arriveTangent=0.0225, interpMode=RCIM_Cubic, leaveTangent=0.0225, tangentMode=RCTM_User, time=0.3, value=1}
              {arriveTangent=-0.4557, interpMode=RCIM_Cubic, leaveTangent=-0.4557, tangentMode=RCTM_User, time=1, value=2}
             ]}}}
          CurveIndex=$Particles.ProgressAlpha
         )
        ))
       B=<1 1 0>
      )
      Vector=&Multiply_VectorByFloat(Vector=$Output.VectorNoiseValue.Noise, Float=10)
     )
     Float=1
    )) {isSetParameters=true}
   /SetVariables(Particles.RibbonLinkOrder:NiagaraFloat=&MakeCustomFloatFromBool(
     Bool=&SetBoolByFloatComparison(A=$Particles.ProgressAlpha, B=0.98)
     True Float=0
     False Float=$Particles.ProgressAlpha
    )) {isSetParameters=true}
   /SetVariables(Particles.GoalPos:NiagaraPosition=&AddVectorToPosition(
     Position=$Particles.SkeletalMeshLocation.SampledPosition
     Vector=&Add_Vector(
      A=&Multiply_VectorByFloat(
       Vector=&Multiply_VectorByFloat(
        Vector=&Add_Vector(
         A=&Lerp_Vector(
          A=<0 0 150>
          B=&FN_FrontEndVector3Modifier(Gameplay=<0 0 350>, Front End=<0 -250 350>)
          Alpha=$System.SmoothDiveAlpha
         )
         B=&Multiply_VectorByFloat(
          Vector=&Lerp_Vector(A=<0 100 0>, B=<0 250 0>, Alpha=$System.SmoothDiveAlpha)
          Float=/Game/Effects/Niagara/DynamicInputs/Math/FN_RemapDotProduct(
           SystemAxis:Vector3f=&TransformVector(Vector=<0 0 1>, Source Space=::World, Destination Space=::World)
           ComparisonAxis:Vector3f=&ReturnCameraVector(Camera Vectors=::NewEnumerator1, Coordinate Space=::Simulation)
           SourceMin:NiagaraFloat=0
           SourceMax:NiagaraFloat=1
           DestMin:NiagaraFloat=0
           DestMax:NiagaraFloat=1
          )
         )
        )
        Float=1
       )
       Float=$Particles.ProgressAlpha
      )
      B=$Particles.RandomOffset
     )
    )) {isSetParameters=true}
   /SetVariables(Particles.Position=/Niagara/DynamicInputs/TypeConversions/MakeCustomPositionFromBool(
     Bool:NiagaraBool=&SetBoolByFloatComparison(A=$Particles.ProgressAlpha, B=0.98)
     True Position:NiagaraPosition=&Add_Vector(A=$Particles.SkeletalMeshLocation.SampledPosition, B=&RandomVector(Vector Scale=20))
     False Position:NiagaraPosition=&Lerp_Vector(
      A=Linked() {linkedVariable {name=Particles.Position, type=NiagaraPosition}}
      B=Linked() {linkedVariable {name=Particles.GoalPos, type=NiagaraPosition}}
      Alpha=&Multiply_Float(
       A=$Engine.DeltaTime
       B=&Lerp_Float(
        A=10
        B=10
        Alpha=&FloatFromCurve(
         FloatCurve=DataInterface() {propertyValues {Curve {keys [
             {arriveTangent=2.9641, interpMode=RCIM_Cubic, leaveTangent=2.9641, tangentMode=RCTM_User, time=0, value=0}
             {arriveTangent=0.3987, interpMode=RCIM_Cubic, leaveTangent=0.3987, tangentMode=RCTM_User, time=1, value=1}
            ]}}}
         CurveIndex=$Particles.ProgressAlpha
        )
       )
      )
     )
    )) {isSetParameters=true}
   &InheritSourceMovement(
    Applied Position Scale=&Multiply_Vector(
     A=&VectorFromFloat(Value=&Multiply_Float(
       A=-0.5
       B=&FloatFromCurve(
        FloatCurve=DataInterface() {propertyValues {Curve {keys [
            {arriveTangent=2.9641, interpMode=RCIM_Cubic, leaveTangent=2.9641, tangentMode=RCTM_User, time=0.1, value=0}
            {arriveTangent=0.3987, interpMode=RCIM_Cubic, leaveTangent=0.3987, tangentMode=RCTM_User, time=0.95, value=1}
            <at 0.97 0 RCIM_Constant>
           ]}}}
        CurveIndex=$Particles.ProgressAlpha
       )
      ))
     B=<1 1 0>
    )
    ApplyForce=false
    ApplyPosition=true
    ApplyPositionFalloff=false
    Source Velocity:Vector3f=&TransformVector(Vector=$Engine.Owner.Velocity, Source Space=::World, Destination Space=::Local)
   )
   &SolveForcesAndVelocity(Clamp Velocity=false, Limit Acceleration=false)
   &self:SetSecondRibbonIdOrder(
    ExecIndex:NiagaraInt32=&ReturnExecIndex()
    Partition:NiagaraInt32=$Particles.PartitionParticles.Partition
    SecondRibbonRibbonOffset=2
    ThirdRibbonOffset=3
   ) {module=SetSecondRibbonIdOrder}
   &Color(Color=&Multiply_LinearColorByFloat(
     Linear Color=&ColorFromCurve(
      ColorCurve=DataInterface() {propertyValues {
        AlphaCurve {keys [<at 0.0131 0>,<at 0.1026 1 RCIM_Cubic>,<at 0.6706 1 RCIM_Cubic>,<at 0.9213 0>]}
        BlueCurve {keys [<at 0.0938 0.2262>,<at 0.2673 0.0679>,<at 0.5041 0.0126>]}
        GreenCurve {keys [<at 0.0938 5>,<at 0.2673 1.5>,<at 0.5041 1>]}
        RedCurve {keys [<at 0.0938 0>,<at 0.2673 0>,<at 0.5041 0>]}
       }}
      CurveIndex=$Particles.ProgressAlpha
     )
     Float=&Lerp_Float(A=1, B=1.5, Alpha=$System.SmoothDiveAlpha)
    ))
   &ScaleRibbonWidth(Ribbon Width Scale=&Add_Float(
     A=&FloatFromCurve(
      FloatCurve=DataInterface() {propertyValues {Curve {keys [
          {arriveTangent=0.2429, interpMode=RCIM_Cubic, leaveTangent=0.2429, tangentMode=RCTM_User, time=0, value=1}
          {interpMode=RCIM_Cubic, tangentMode=RCTM_User, time=0.8, value=1}
          {interpMode=RCIM_Cubic, tangentMode=RCTM_User, time=1, value=0}
         ]}}}
      CurveIndex=$Particles.ProgressAlpha
      Scale Curve=&Lerp_Float(A=3, B=4, Alpha=$System.SmoothDiveAlpha)
     )
     B=&Multiply_Float(
      A=4
      B=&FloatFromCurve(
       FloatCurve=DataInterface() {propertyValues {Curve {keys [
           {interpMode=RCIM_Cubic, tangentMode=RCTM_User, time=0, value=0}
           {interpMode=RCIM_Cubic, tangentMode=RCTM_User, time=0.1, value=1}
           {interpMode=RCIM_Cubic, tangentMode=RCTM_User, time=0.4, value=0}
           {interpMode=RCIM_Cubic, tangentMode=RCTM_User, time=0.7, value=0}
           {interpMode=RCIM_Cubic, tangentMode=RCTM_User, time=0.8, value=0}
          ]}}}
       CurveIndex=&ModuloFloat(
        A=&Abs_Float(Float:NiagaraFloat=&Add_Float(
          A=&Add_Float(
           A=$Particles.PartitionParticles.Normalized Index In Partition
           B=&Multiply_FloatByInt(Float=0.5, Integer=$Particles.PartitionParticles.Partition)
          )
          B=&Multiply_Float(A=$Particles.Age, B=-0.25)
         ))
        B=1
       )
      )
     )
    ))
   /SetVariables(Particles.SecondRibbonWidth:NiagaraFloat=&Multiply_Float(A=$Particles.RibbonWidth, B=0.7)) {isSetParameters=true}
   /SetVariables(Particles.SecondRibbonColor:LinearColor=&MultiplyLinearColor(Linear Color=$Particles.Color, Scale Factor=<1 1 1 0.5>)) {isSetParameters=true}
   /SetVariables(Particles.ThirdRibbonWidth:NiagaraFloat=&Multiply_Float(A=$Particles.RibbonWidth, B=0.5)) {isSetParameters=true}
   /SetVariables(Particles.ThirdRibbonColor:LinearColor=&MultiplyLinearColor(Linear Color=$Particles.Color, Scale Factor=<1 1 1 0.2>)) {isSetParameters=true}
   /SetVariables(Particles.GlowWidth:NiagaraFloat=&Multiply_Float(A=$Particles.RibbonWidth, B=4)) {isSetParameters=true}
   /SetVariables(Particles.GlowColor=&ColorFromCurve(
     ColorCurve=DataInterface() {propertyValues {
       AlphaCurve {keys [<at 0.0682 0.05 RCIM_Cubic>,<at 0.7109 0 RCIM_Cubic>]}
       BlueCurve {keys [<at 0.0781 0.0452 RCIM_Cubic>,<at 1 1 RCIM_Cubic>]}
       GreenCurve {keys [<at 0.0781 1 RCIM_Cubic>,<at 1 1 RCIM_Cubic>]}
       RedCurve {keys [<at 0.0781 0 RCIM_Cubic>,<at 1 1 RCIM_Cubic>]}
      }}
     CurveIndex=$Particles.ProgressAlpha
    )) {isSetParameters=true}
   &UpdateMeshOrientation(
    Rotation Rate=&RandomRangeFloat(Minimum=0.2, Maximum=-0.2)
    Rotation Vector=&RandomVector(Vector Scale=1)
   )
   &ScaleMeshSize(Scale Factor=&VectorFromFloat(Value=&Multiply_Float(
      A=&FloatFromCurve(
       FloatCurve=DataInterface() {propertyValues {Curve {keys [
           {arriveTangent=3.195, interpMode=RCIM_Cubic, leaveTangent=3.195, tangentMode=RCTM_User, time=0, value=0}
           {interpMode=RCIM_Cubic, tangentMode=RCTM_User, time=0.5, value=1}
           {arriveTangent=-0.0572, interpMode=RCIM_Cubic, leaveTangent=-0.0572, tangentMode=RCTM_User, time=1, value=0.35}
          ]}}}
       CurveIndex=$Particles.ProgressAlpha
       Scale Curve=&Lerp_Float(A=0.8, B=2, Alpha=$System.SmoothDiveAlpha)
      )
      B=&Lerp_Float(
       A=0
       B=1
       Alpha=&FloatFromCurve(
        FloatCurve=DataInterface() {propertyValues {Curve {keys [<at 0.5 0 RCIM_Constant>,<at 0.95 1 RCIM_Constant>]}}}
        CurveIndex=$Particles.RandomFloatPerCycle
       )
      )
     )))
   /SetVariables(Particles.SphereDynMat:Vector4f=&MakeVector4(
     X=&FloatFromCurve(
      FloatCurve=DataInterface() {propertyValues {Curve {keys [
          {arriveTangent=0.2592, interpMode=RCIM_Cubic, leaveTangent=0.2592, tangentMode=RCTM_User, time=0.25, value=0}
          {arriveTangent=1.2, interpMode=RCIM_Cubic, leaveTangent=1.2, tangentMode=RCTM_User, time=1, value=1}
         ]}}}
      CurveIndex=$Particles.ProgressAlpha
     )
     Y=&RandomRangeFloat(Minimum=50, Maximum=70)
     Z=&FloatFromCurve(
      FloatCurve=DataInterface() {propertyValues {Curve {keys [
          {interpMode=RCIM_Cubic, tangentMode=RCTM_User, time=0, value=1}
          {arriveTangent=-0.1731, interpMode=RCIM_Cubic, leaveTangent=-0.1731, tangentMode=RCTM_User, time=0.95, value=0}
         ]}}}
      CurveIndex=$Particles.ProgressAlpha
      Scale Curve=1.5
     )
     W=0
    )) {isSetParameters=true}
   /SetVariables(Particles.UnrollDirection:Vector3f=&FN_FrontEndVector3Modifier(
     Gameplay=&TransformVector(
      Vector=&Lerp_Vector(
       A=$Particles.UnrollDirection
       B=&NormalizeVector(
        Vector To Normalize=&Multiply_VectorByFloat(Vector=/NiagaraFluids/DynamicInput/MaxVectorByLength(A:Vector3f=<0 0 -0.1>, B:Vector3f=$Engine.Owner.Velocity), Float=-1)
        Safe Normalize=true
        Safe Fallback Threshold=0
        Safe Fallback Vector=<0 0 0.001>
       )
       Alpha=&Multiply_Float(A=$Engine.DeltaTime, B=0.5)
      )
      Source Space=::World
      Destination Space=::Local
     )
     Front End=&Lerp_Vector(A=<0 0 1>, B=<0 -1 1>, Alpha=$System.SmoothDiveAlpha)
    )) {isSetParameters=true}
   &DynamicMaterialParameters(
    Index 1 Param 1=&MakeFloatFromVector(Vector=$Particles.UnrollDirection, Channel=::NewEnumerator0)
    Index 1 Param 2=&MakeFloatFromVector(Vector=$Particles.UnrollDirection, Channel=::NewEnumerator1)
    Index 1 Param 3=&MakeFloatFromVector(Vector=$Particles.UnrollDirection, Channel=::NewEnumerator2)
    Module.Index1Param0Write=true
    Module.Index1Param1Write=true
    Module.Index1Param2Write=true
    Module.Index1Param3Write=true
    Write Parameter Index 0=false
    Write Parameter Index 1=true
   )
   /SetVariables(Particles.MeshOrientation=&OrientMeshToVector(Look At Direction=$Particles.UnrollDirection, Reference Axis Vector=<0 1 0>)) {isSetParameters=true}
   /SetVariables(
    Particles.Color=&MakeCustomLinearColorFromBool(
     Bool=&Boolean_OrOperation(Bool1=$User.bIsFrontEnd, Bool2=$User.bIsFrontEndPreview)
     True Color=&ColorFromCurve(
      ColorCurve=DataInterface() {propertyValues {
        AlphaCurve {keys [<at 0.0131 0>,<at 0.1026 1 RCIM_Cubic>,<at 0.6706 1 RCIM_Cubic>,<at 0.9213 0>]}
        BlueCurve {keys [<at 0.0938 0.0679>,<at 0.2673 0.0226>,<at 0.5041 0.0063>]}
        GreenCurve {keys [<at 0.0938 1.5>,<at 0.2673 0.5>,<at 0.5041 0.5>]}
        RedCurve {keys [<at 0.0938 0>,<at 0.2673 0>,<at 0.5041 0>]}
       }}
      CurveIndex=$Particles.ProgressAlpha
     )
     False Color=$Particles.Color
    )
    Particles.SecondRibbonColor:LinearColor=&MakeCustomLinearColorFromBool(
     Bool=$User.bIsFrontEnd
     True Color=&ColorFromCurve(
      ColorCurve=DataInterface() {propertyValues {
        AlphaCurve {keys [<at 0.0131 0>,<at 0.1026 1 RCIM_Cubic>,<at 0.6706 1 RCIM_Cubic>,<at 0.9213 0>]}
        BlueCurve {keys [<at 0.0938 0.0679>,<at 0.2673 0.0226>,<at 0.5041 0.0063>]}
        GreenCurve {keys [<at 0.0938 1.5>,<at 0.2673 0.5>,<at 0.5041 0.5>]}
        RedCurve {keys [<at 0.0938 0>,<at 0.2673 0>,<at 0.5041 0>]}
       }}
      CurveIndex=$Particles.ProgressAlpha
     )
     False Color=$Particles.SecondRibbonColor
    )
    Particles.ThirdRibbonColor:LinearColor=&MakeCustomLinearColorFromBool(
     Bool=$User.bIsFrontEnd
     True Color=&ColorFromCurve(
      ColorCurve=DataInterface() {propertyValues {
        AlphaCurve {keys [<at 0.0131 0>,<at 0.1026 1 RCIM_Cubic>,<at 0.6706 1 RCIM_Cubic>,<at 0.9213 0>]}
        BlueCurve {keys [<at 0.0938 0.0679>,<at 0.2673 0.0226>,<at 0.5041 0.0063>]}
        GreenCurve {keys [<at 0.0938 1.5>,<at 0.2673 0.5>,<at 0.5041 0.5>]}
        RedCurve {keys [<at 0.0938 0>,<at 0.2673 0>,<at 0.5041 0>]}
       }}
      CurveIndex=$Particles.ProgressAlpha
     )
     False Color=$Particles.ThirdRibbonColor
    )
   ) {isSetParameters=true}
  ]
  rendererClasses [&NiagaraRibbonRendererProperties, &NiagaraMeshRendererProperties]
  renderers [
   {rendererClass=&NiagaraRibbonRendererProperties}
   {rendererClass=&NiagaraRibbonRendererProperties, rendererIndex=1}
   {rendererClass=&NiagaraRibbonRendererProperties, rendererIndex=2}
   {rendererClass=&NiagaraMeshRendererProperties, rendererIndex=3}
   {rendererClass=&NiagaraMeshRendererProperties, rendererIndex=4}
  ]
 }
}
summary {
 systemName=NS_Contrail_SliceVine
 emitters [
  {
   emitterName=NE_Balls
   rendererClasses [
    &NiagaraSpriteRendererProperties
    &NiagaraMeshRendererProperties
    &NiagaraLightRendererProperties
   ]
  }
  {emitterName=NE_PlexusTrail, rendererClasses [&NiagaraRibbonRendererProperties, &NiagaraMeshRendererProperties]}
  {emitterName=Flares, rendererClasses [&NiagaraSpriteRendererProperties]}
 ]
 userVariables [
  User.bIsFrontEnd {type=NiagaraBool}
  User.bIsFrontEndPreview {type=NiagaraBool}
  User.ParaGlideLeanAlpha {type=NiagaraFloat, defaultValue=0, description=""}
 ]
}
